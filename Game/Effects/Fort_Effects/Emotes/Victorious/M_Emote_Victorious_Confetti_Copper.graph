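# Copper confetti material for the Victorious emote: masked, two-sided, lit and built for Niagara sprites, mesh particles and ribbons.
# Confetti mask and normal textures share UVs warped by a radial flowmap, with alpha-driven erosion.
# LackLuster base attributes get a copper tint, alongside Fresnel, swirl-normal, curve-atlas iridescence and hue shifts.
# Dynamic parameters control warp, mesh bend, emissive boost and front-end color and normal adjustments.
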Material /Game/Effects/Fort_Effects/Emotes/Victorious/M_Emote_Victorious_Confetti_Copper
graph {
 properties {
  bUsedWithNiagaraMeshParticles=true
  bUsedWithNiagaraRibbons=true
  bUsedWithNiagaraSprites=true
  blendMode=BLEND_Masked
  pixelDepthOffsetMode=PDOM_Legacy
  twoSided=true
  usedShadingModels=Default Lit
 }
 nodes [
  n1 = NamedRerouteUsage()
  n2 = ParticleColor()
  n3 = BreakMaterialAttributes!(Attr=MaterialFunctionCall!(
    bBlendMaterialAttributes
    MA_In
    UVsOverride
    Override Normals
    BlendMaterialAttributesAlpha
    BaseColorOverride=@n2.RGB
    WorldPositionOverride
    VertexNormalOverride
    SourceNormalsStrengthOverride
   ) {materialFunction=/Game/TechArt/Seasonal/LackLuster/MaterialFunctions/MF_LackLuster_Base}.MA Out)
  Add(A=@n1, B=@n3.BaseColor)
  n4 = VectorParameter() {parameterName=Tile, defaultValue=<rgba 1 1 0 0>}
  n5 = MaterialFunctionCall(Texture Scale=Constant() {r=1.4}) {materialFunction=&ScaleUVsByCenter}
  n6 = Subtract(A=ComponentMask(None=TextureSample(UVs=@n5.UVs) {samplerType=SAMPLERTYPE_LinearColor, texture=/Game/Environments/Hexmap/HexClouds/T_RadialFlowmap01}.RGB) {g=true, r=true}) {constB=0.5}
  n7 = DynamicParameter() {parameterIndex=1, defaultValue=<rgba 0 0 0 1>, paramNames [CenterWarp, MeshWPOBend, forFrontEnd, EmissiveAdd Scale]}
  n8 = Multiply(A=AppendVector(A=@n4.R, B=@n4.R), B=Add(A=TextureCoordinate(), B=Multiply(A=@n6, B=@n7.CenterWarp)))
  n9 = TextureSampleParameter2D(UVs=@n8) {parameterName=Mask, texture=/Game/Effects/Fort_Effects/Textures/Confetti/T_Confetti_SubUV_4X4_A}
  ChannelMaskParameter(None=@n9.RGBA) {parameterName=OpacityMask}
  Constant3Vector() {constant=<rgba 0 0 1 0>}
  n10 = TextureSampleParameter2D(UVs=@n8) {parameterName=Normal, samplerType=SAMPLERTYPE_Normal, texture=/Game/Effects/Fort_Effects/Textures/Confetti/T_Confetti_SubUV_4X4_03}
  n11 = DynamicParameter() {defaultValue=<rgba 0 0 1 0>, paramNames [X, Y, Z, FrontEndNormalAdjust]}
  LinearInterpolate(
   A=@n10.RGB
   B=Constant3Vector() {constant=<rgba 0 0 1 0>}
   Alpha=Add(A=ScalarParameter() {defaultValue=0.92, parameterName=Flatten Normal}, B=@n11.FrontEndNormalAdjust)
  )
  n12 = LinearInterpolate(B=@n2.RGB, Alpha=ScalarParameter() {defaultValue=1, parameterName=UseParticleColor})
  n13 = DynamicParameter() {parameterIndex=1, defaultValue=<rgba 0 0 0 1>, paramNames [CenterWarp, MeshWPOBend, forFrontEnd, EmissiveAdd Scale]}
  n14 = Multiply(A=Multiply(A=@n13.EmissiveAdd Scale, B=ScalarParameter() {defaultValue=0.01, parameterName=EmissiveAdd}), B=NamedRerouteUsage())
  MaterialFunctionCall(Emissive=Add(A=Multiply(A=@n12, B=@n14) {constB=0.15}, B=Constant3Vector() {constant=<rgba 0.0365 0.018 0.0078 1>})) {materialFunction=&MF_FX_EmissiveAdjustments}
  n15 = ChannelMaskParameter(None=@n9.RGB) {maskChannel=Blue, parameterName=AlphaMask_Channel, defaultValue=<rgba 0 0 1 0>}
  n16 = ChannelMaskParameter(None=@n15.R) {maskChannel=Blue, parameterName=ErodeMask, defaultValue=<rgba 0 0 1 0>}
  n17 = LinearInterpolate(Alpha=@n2.A) {constA=2, constB=-0.1}
  Multiply(A=Multiply(A=@n2.A, B=Power(Base=@n16.R, Exp=@n17)), B=@n15.R)
  n18 = DynamicParameter() {parameterIndex=1, defaultValue=<rgba 0 0 0 1>, paramNames [CenterWarp, MeshWPOBend, forFrontEnd, EmissiveAdd Scale]}
  Multiply(A=LinearInterpolate(A=@n3.BaseColor, B=@n1, Alpha=@n18.forFrontEnd), B=Constant3Vector() {constant=<rgba 0.9479 0.6952 0.4147 1>})
  n19 = ParticleDirection()
  n20 = Multiply(A=@n19, B=@n19)
  n21 = TextureCoordinate() {uTiling=11}
  n22 = ParticleRandom()
  n23 = Frac(None=ComponentMask(None=Add(A=@n21, B=@n22)) {r=true})
  Multiply(
   A=MaterialFunctionCall(Hue Shift Percentage=Add(A=@n20, B=@n23), Texture=Constant3Vector() {constant=<rgba 1 0 1 0>}) {materialFunction=&HueShift}.Result
   B=VectorParameter() {parameterName=FE Color Adjustment, defaultValue=<rgba 0 0.0645 0.9635 0>}.RGB
  )
  n24 = TextureCoordinate() {uTiling=0.01, vTiling=0.002}
  n25 = TextureSample(UVs=@n24) {
   samplerSource=SSM_Wrap_WorldGroupSettings
   samplerType=SAMPLERTYPE_Normal
   texture=/Game/TechArt/Seasonal/LackLuster/Textures/T_Lackluster_Swirl_N
  }
  n26 = TextureSample(UVs=@n24) {
   samplerSource=SSM_Wrap_WorldGroupSettings
   samplerType=SAMPLERTYPE_Normal
   texture=&T_Lackluster_Swirl_N
  }
  n27 = Add(A=@n25.RGB, B=@n26.RGB)
  n28 = Transform(None=Multiply(A=VectorParameter() {parameterName=EdgeNormalScale, defaultValue=<rgba 1 1 0 0>}.RGB, B=Normalize(VectorInput=@n27)))
  n29 = LinearInterpolate(Alpha=MaterialFunctionCall() {materialFunction=&MF_Fresnel_FX}.Fresnel Inverted) {constA=0.05}
  n30 = TextureSample(UVs=Add(A=ComponentMask(None=@n28) {g=true, r=true}, B=@n29)) {samplerType=SAMPLERTYPE_LinearColor, texture=/Game/TechArt/Seasonal/LackLuster/Textures/T_Lackluster_Streaks}
  n31 = CurveAtlasRowParameter(CurveTime=Frac(None=Add(A=@n30.R))) {
   atlas=/Game/TechArt/Seasonal/LackLuster/Curves/CA_Lackluster
   curve=/Game/TechArt/Seasonal/LackLuster/Curves/C_LackLuster_Spectrum_04
   defaultValue=8
   parameterName=ReflectionCurve
  }
  Multiply(A=@n31, B=@n29)
  Multiply(A=@n3.Metallic, B=ScalarParameter() {defaultValue=1, parameterName=Metallic})
  Multiply(A=@n3.Specular, B=ScalarParameter() {defaultValue=0.5, parameterName=Specular})
  n32 = ComponentMask(None=TextureCoordinate()) {g=true}
  Multiply(A=Multiply(A=@n11.RGB, B=@n7.MeshWPOBend), B=Power(Base=OneMinus(None=@n32)) {constExponent=3})
  Multiply(A=@n3.Roughness, B=ScalarParameter() {defaultValue=0.2, parameterName=Roughness})
  n33 = VertexNormalWS()
  n34 = ComponentMask(None=TextureCoordinate()) {r=true}
  NamedRerouteDeclaration(None=MaterialFunctionCall(Hue Shift Percentage=Add(A=@n33, B=@n34), Texture=Constant3Vector() {constant=<rgba 1 0.8 0.8 0>}) {materialFunction=&HueShift}.Result) {name=Hue Color, nodeColor=<rgba 1 0 0.5391 1>}
  NamedRerouteDeclaration(None=Power(Base=@n31) {constExponent=1.5}) {name=forFE, nodeColor=<rgba 0.3359 0 1 1>}
  TextureCoordinate() {uTiling=0.02, vTiling=0.05}
 ]
}
